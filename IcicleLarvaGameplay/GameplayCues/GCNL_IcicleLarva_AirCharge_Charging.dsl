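; Looping gameplay cue that plays the Icicle Larva's air charge audio while it is charging.
; On loop start it attaches start and movement loop sounds to the pawn mesh in custom logic, because default cue data does not attach correctly.
; It stores the loop component and fades it out on removal.
; A looping targeted sound indicator runs alongside, skipped when the pawn is locally controlled or viewed.
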(event EventOnLoopingStart(EitherVFX-preferthisversion)
  (params (MyTarget :type Actor%20Object%20Reference) (Parameters :type Gameplay%20Cue%20Parameters%20Structure%20%28by%20ref%29) (ParticleComponents :type Array%20of%20FXSystem%20Component%20Object%20References) (AudioComponents :type Array%20of%20Audio%20Component%20Object%20References))
  (|CallParentFunction MyTarget Parameters ParticleComponents AudioComponents)
  (CallFunction|PlayFlyingChargeLoop true MyTarget)
  ; Don't show AV for Icicle Larva player
  (bind _asfort_pawn (Utilities|Casting|CastToFortPawn MyTarget))
  (if (not (or (Pawn|IsLocallyControlled _asfort_pawn) (Pawn|IsLocallyViewed _asfort_pawn)))
    (SoundIndicator|StartTargetedSoundIndicatorLooping (Variables|Default|GetFortSoundIndicator) MyTarget 0 1.0 0)))

(event EventOnRemoval(EitherVFX-preferthisversion)
  (params (MyTarget :type Actor%20Object%20Reference) (Parameters :type Gameplay%20Cue%20Parameters%20Structure%20%28by%20ref%29) (ParticleComponents :type Array%20of%20FXSystem%20Component%20Object%20References) (AudioComponents :type Array%20of%20Audio%20Component%20Object%20References) (BurstCameraShakeInstance :type Camera%20Shake%20Base%20Object%20Reference) (BurstDecalInstance :type Decal%20Actor%20Object%20Reference))
  (|CallParentFunction MyTarget Parameters ParticleComponents AudioComponents BurstCameraShakeInstance BurstDecalInstance)
  (CallFunction|PlayFlyingChargeLoop false MyTarget)
  ; Don't show AV for Icicle Larva player
  (bind _asfort_pawn (Utilities|Casting|CastToFortPawn MyTarget))
  (if (not (or (Pawn|IsLocallyControlled _asfort_pawn) (Pawn|IsLocallyViewed _asfort_pawn)))
    (SoundIndicator|StopSoundIndicatorLooping (Variables|Default|GetFortSoundIndicator))))

(event Custom|PlayFlyingChargeLoop
  (params (bPlay :type Boolean) (PawnRef :type Actor%20Object%20Reference))
  ; CustomAudio playback - default data is not attaching correctly
  (if bPlay
    (bind _asfort_pawn (Utilities|Casting|CastToFortPawn PawnRef))
    (bind _mesh (Variables|/Script/Engine.Character|GetMesh _asfort_pawn)
      (:then
        (Audio|SpawnSoundAttached (Variables|Default|GetMovementStartSound) _mesh)
        (bind _returnvalue (Audio|SpawnSoundAttached :Sound (Variables|Default|GetMovementSound) :AttachToComponent _mesh :AttachPointName "None" :LocationType "KeepRelativeOffset" :bStopWhenAttachedToDestroyed true :VolumeMultiplier 1.0 :PitchMultiplier 1.0 :StartTime 0.0 :bAutoDestroy true))
        (Variables|Default|SetLoopAudioComponent _returnvalue))
      (:else))
    (else
      (bind _loopaudiocomponent (Variables|Default|GetLoopAudioComponent)
        (:then
          (Audio|Components|Audio|FadeOut _loopaudiocomponent 2.0 0.0 "Linear" :!class /Script/Engine.AudioComponent))
        (:else)))))

(fn ConstructionScript)
