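; Looping gameplay cue that plays flying audio while an Icicle Larva hovers.
; On start, PlayFlyingLoop fades any prior loop, then attaches an optional start sound and movement loop to the pawn mesh, working around default cue data attaching incorrectly.
; Removal and EndPlay fade out the stored loop.
; A looping targeted sound indicator follows the cue's start and removal, skipped when the larva is locally controlled or viewed.
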(event EventOnLoopingStart(EitherVFX-preferthisversion)
  (params (MyTarget :type Actor%20Object%20Reference) (Parameters :type Gameplay%20Cue%20Parameters%20Structure%20%28by%20ref%29) (ParticleComponents :type Array%20of%20FXSystem%20Component%20Object%20References) (AudioComponents :type Array%20of%20Audio%20Component%20Object%20References))
  (|CallParentFunction MyTarget Parameters ParticleComponents AudioComponents)
  (CallFunction|PlayFlyingLoop MyTarget true)
  ;[1 1 1] Don't show AV for Icicle Larva player
  (bind _asfort_pawn (Utilities|Casting|CastToFortPawn MyTarget))
  (if (not (or (Pawn|IsLocallyControlled _asfort_pawn) (Pawn|IsLocallyViewed _asfort_pawn)))
    (SoundIndicator|StartTargetedSoundIndicatorLooping (Variables|Default|GetFortSoundIndicator) MyTarget 0 1.0 0)))

(event EventOnRemoval(EitherVFX-preferthisversion)
  (params (MyTarget :type Actor%20Object%20Reference) (Parameters :type Gameplay%20Cue%20Parameters%20Structure%20%28by%20ref%29) (ParticleComponents :type Array%20of%20FXSystem%20Component%20Object%20References) (AudioComponents :type Array%20of%20Audio%20Component%20Object%20References) (BurstCameraShakeInstance :type Camera%20Shake%20Base%20Object%20Reference) (BurstDecalInstance :type Decal%20Actor%20Object%20Reference))
  (|CallParentFunction MyTarget Parameters ParticleComponents AudioComponents BurstCameraShakeInstance BurstDecalInstance)
  (CallFunction|PlayFlyingLoop MyTarget)
  ;[1 1 1] Don't show AV for Icicle Larva player
  (bind _asfort_pawn (Utilities|Casting|CastToFortPawn MyTarget))
  (if (not (or (Pawn|IsLocallyControlled _asfort_pawn) (Pawn|IsLocallyViewed _asfort_pawn)))
    (SoundIndicator|StopSoundIndicatorLooping (Variables|Default|GetFortSoundIndicator))))

(event Custom|PlayFlyingLoop
  (params (Actor :type Actor%20Object%20Reference) (bPlay :type Boolean))
  ;[1 1 1] CustomAudio playback - default data is not attaching correctly
  (if bPlay
    (bind _asfort_pawn (Utilities|Casting|CastToFortPawn Actor))
    (bind _loopaudiocomponent (Variables|Default|GetLoopAudioComponent)
      (:then
        (Audio|Components|Audio|FadeOut _loopaudiocomponent 0.15 0.0 "Linear" :!class /Script/Engine.AudioComponent))
      (:else
        _))
    (bind _mesh (Variables|/Script/Engine.Character|GetMesh _asfort_pawn)
      (:then
        (bind _movementstartsound (Variables|Default|GetMovementStartSound)
          (:then
            (Audio|SpawnSoundAttached _movementstartsound _mesh))
          (:else
            _))
        (bind _returnvalue (Audio|SpawnSoundAttached :Sound (Variables|Default|GetMovementSound) :AttachToComponent _mesh :AttachPointName "None" :LocationType "KeepRelativeOffset" :bStopWhenAttachedToDestroyed true :VolumeMultiplier 1.0 :PitchMultiplier 1.0 :StartTime 0.0 :bAutoDestroy true))
        (Variables|Default|SetLoopAudioComponent _returnvalue))
      (:else))
    (else
      (bind _loopaudiocomponent_2 (Variables|Default|GetLoopAudioComponent)
        (:then
          (Audio|Components|Audio|FadeOut _loopaudiocomponent_2 0.5 0.0 "Linear" :!class /Script/Engine.AudioComponent))
        (:else)))))

(event EventEndPlay
  (params (EndPlayReason :type EEndPlayReason%20Enum :default "Destroyed"))
  (|CallParentFunction EndPlayReason)
  (bind _loopaudiocomponent (Variables|Default|GetLoopAudioComponent)
    (:then
      (Audio|Components|Audio|FadeOut _loopaudiocomponent 0.15 0.0 "Linear" :!class /Script/Engine.AudioComponent))
    (:else)))

(fn ConstructionScript)
