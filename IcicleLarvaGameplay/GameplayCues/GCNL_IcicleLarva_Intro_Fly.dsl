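; Burst gameplay cue for the Icicle Larva's intro fly, adding a sound indicator for the event.
; On burst it shows a local, team-filtered Generic sound indicator that tracks the target actor from its location, with a long audible range.
; The indicator uses Neutral and Hostile affiliations, and the team lookup does not read from the target.
; It then calls the parent OnBurst and returns that result.
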(fn OnBurst
  (params (MyTarget :type Actor%20Object%20Reference) (Parameters :type Gameplay%20Cue%20Parameters%20Structure%20%28by%20ref%29))
  (Fortnite|ShowSoundIndicatorLocalforTeam :WorldContextObject MyTarget :TrackedActor MyTarget :Location (Transformation|GetActorLocation MyTarget) :MaxAudibleDistance 1.08e+04 :IndicatorType "Generic" :Team (Team|GetActorTeam 0) :Affiliations (Utilities|Array|MakeArray "Neutral" "Hostile") :Instigator MyTarget :Tint "(R=0.000000,G=0.000000,B=0.000000,A=0.000000)" :AttenuationMaxRangeMultiplier 1.0 :AttenuationIntensityMultiplier 1.0)
  (bind _returnvalue (|CallParentFunction MyTarget Parameters))
  (return _returnvalue))

(fn ConstructionScript)
