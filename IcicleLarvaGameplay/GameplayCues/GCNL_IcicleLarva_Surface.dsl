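; Looping gameplay cue that plays surface particles and ground audio for the Icicle Larva on the target player pawn.
; Each tick it reads surface hit data from the pawn's IcicleLarva component and, above a speed threshold, interpolates toward the impact point and normal-aligned rotation.
; It repositions particles manually as an attachment workaround and switches particles and MetaSound for water hits.
; Activation spawns a radial force field if physics fields are enabled; slowdown or removal destroys it.
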(event EventOnLoopingStart(EitherVFX-preferthisversion)
  (params (MyTarget :type Actor%20Object%20Reference) (Parameters :type Gameplay%20Cue%20Parameters%20Structure%20%28by%20ref%29) (ParticleComponents :type Array%20of%20FXSystem%20Component%20Object%20References) (AudioComponents :type Array%20of%20Audio%20Component%20Object%20References))
  (Variables|Default|SetParticleComponents ParticleComponents)
  (bind _asfort_player_pawn (Utilities|Casting|CastToFortPlayerPawn MyTarget)
    (:then
      (bind _output_get (Variables|Default|SetTargetPlayerPawn _asfort_player_pawn))
      (Variables|Default|SetIcicleLarvaPawnComponent (Actor|GetComponentbyClass _output_get "/Script/IcicleLarvaGameplayRuntime.FortPawnComponent_IcicleLarva")))
    (:CastFailed
      (Fortnite|FortShippingLog self "Target of Surface GCNL is not a valid FortPlayerPawn!" true))))

(event EventTick
  (params (DeltaSeconds :type Float%20%28single-precision%29))
  (bind (_returnvalue_location _returnvalue_rotation) (Math|Transform|BreakTransform (Transformation|GetActorTransform)))
  (bind _iciclelarvapawncomponent (Variables|Default|GetIcicleLarvaPawnComponent))
  (bind (_surfacehitdata_bmadeimpact _surfacehitdata_hitactor _surfacehitdata_bhitwater _surfacehitdata_hitimpactpoint _surfacehitdata_hitnormal _surfacehitdata_ownerrotation) (Utilities|Struct|BreakSurfaceHitData (Variables|/Script/IcicleLarvaGameplayRuntime.FortPawnComponent_IcicleLarva|GetSurfaceHitData_bMadeImpact _iciclelarvapawncomponent)))
  (bind _output_get (Variables|Default|SetCurrentVelocity (Math|Vector|VectorLength (Transformation|GetVelocity (Variables|Default|GetTargetPlayerPawn)))))
  (if (>= _output_get (ScalableFloat|GetValueatLevel (Variables|Default|GetSurfaceFXMinSpeedToBeActive)))
    (CallFunction|ActivateSurfaceFX)
    (bind _output_get_2 (Variables|Default|SetHit%20Water _surfacehitdata_bhitwater))
    (CallFunction|UpdateSurfaceAudio _output_get_2)
    (bind _rotation (CallFunction|CalculateRotation _surfacehitdata_hitnormal _surfacehitdata_ownerrotation))
    (Transformation|SetActorLocationAndRotation (Math|Interpolation|VInterptoConstant (Transformation|GetActorLocation) _surfacehitdata_hitimpactpoint DeltaSeconds (* (Variables|Default|GetCurrentVelocity) (ScalableFloat|GetValueatLevel (Variables|Default|GetSurfaceFXLocationInterpSpeedMultiplier)))) (Math|Interpolation|RInterpTo (Transformation|GetActorRotation) _rotation DeltaSeconds (ScalableFloat|GetValueatLevel (Variables|Default|GetSurfaceFXRotationInterpSpeed))))
    (for _array_element (Variables|Default|GetParticleComponents)
      ; For some reason no matter what I do the particle components don't attach to root, so I'm manually updating their transform instead because I don't have time to investigate further
      (Transformation|SetWorldLocationAndRotation _array_element (+ _returnvalue_location (* (Math|Vector|GetForwardVector _returnvalue_rotation) (Math|Conversions|ToVector(Float) (Variables|Default|GetForwardOffset)))) _returnvalue_rotation false false)
      (Effects|Components|ParticleSystem|SetFloatParameter _array_element "WaterSpawnRate" (Math|Float|SelectFloat 1.0 0.0 (Variables|Default|GetHit%20Water))))
    (CallFunction|DrawDebugShapes)
    (else
      (CallFunction|DeactivateSurfaceFX)
      (CallFunction|PlaySurfaceAudio false))))

(event EventOnRemoval(EitherVFX-preferthisversion)
  (params (MyTarget :type Actor%20Object%20Reference) (Parameters :type Gameplay%20Cue%20Parameters%20Structure%20%28by%20ref%29) (ParticleComponents :type Array%20of%20FXSystem%20Component%20Object%20References) (AudioComponents :type Array%20of%20Audio%20Component%20Object%20References) (BurstCameraShakeInstance :type Camera%20Shake%20Base%20Object%20Reference) (BurstDecalInstance :type Decal%20Actor%20Object%20Reference))
  (bind _radialforceactor (Variables|Default|GetRadialForceActor))
  (CallFunction|PlaySurfaceAudio false)
  (Utilities|IsValid _radialforceactor
    (:Is%20Valid
      (Actor|DestroyActor _radialforceactor))))

(event Custom|ActivateSurfaceFX
  (bind _targetplayerpawn (Variables|Default|GetTargetPlayerPawn))
  (bind _iciclelarvapawncomponent (Variables|Default|GetIcicleLarvaPawnComponent))
  (bind (_surfacehitdata_bmadeimpact _surfacehitdata_hitactor _surfacehitdata_bhitwater _surfacehitdata_hitimpactpoint _surfacehitdata_hitnormal _surfacehitdata_ownerrotation) (Utilities|Struct|BreakSurfaceHitData (Variables|/Script/IcicleLarvaGameplayRuntime.FortPawnComponent_IcicleLarva|GetSurfaceHitData_bMadeImpact _iciclelarvapawncomponent)))
  (if (not (Variables|Default|GetbSurfaceFXActive))
    (CallFunction|PlaySurfaceAudio true)
    (bind _rotation (CallFunction|CalculateRotation _surfacehitdata_hitnormal _surfacehitdata_ownerrotation))
    (Transformation|SetActorLocationAndRotation _surfacehitdata_hitimpactpoint _rotation)
    (Variables|Default|SetbSurfaceFXActive true)
    (for _array_element (Variables|Default|GetParticleComponents)
      (Components|Activation|Activate _array_element true)
      (Development|PrintString 0 "Activate"))
    (bind _returnvalue (Development|GetConsoleVariableIntValue "r.PhysicsField.EnableField"))
    (if (== _returnvalue 1)
      (Utilities|IsValid (Variables|Default|GetRadialForceActor)
        (:Is%20Not%20Valid
          (bind _returnvalue_2 (Game|SpawnActorfromClass _targetplayerpawn "/Game/Effects/Niagara/Blueprints/PhsyicsFields/FN_RadialForce.FN_RadialForce_C" (Math|Conversions|ToTransform(Vector) (Transformation|GetActorLocation _targetplayerpawn)) "AlwaysSpawn" "MultiplyWithRoot" 0 0.2 5e+03 1e+03 1.5 0.0 0))
          (bind _output_get (Variables|Default|SetRadialForceActor _returnvalue_2))
          (Transformation|AttachActorToComponent _output_get (Variables|/Script/Engine.Actor|GetRootComponent _targetplayerpawn) "None" "SnapToTarget" "SnapToTarget" "SnapToTarget" false))))))

(event Custom|DeactivateSurfaceFX
  (Variables|Default|SetbSurfaceFXActive false)
  (for _array_element (Variables|Default|GetParticleComponents)
    (Components|Activation|Deactivate _array_element)
    (Development|PrintString 0 "Deactivate"))
  (Utilities|IsValid (Variables|Default|GetRadialForceActor)
    (:Is%20Valid
      (Actor|DestroyActor (Variables|Default|GetRadialForceActor)))))

(event Custom|PlaySurfaceAudio
  (params (bPlay :type Boolean))
  (if bPlay
    (bind _rootcomponent (Variables|Transformation|GetRootComponent)
      (:then
        (Variables|Default|GetSurfaceAudioComponent
          (:then)
          (:else
            (bind _returnvalue (Audio|SpawnSoundAttached "/IcicleLarvaGameplay/Assets/Audio/MetaSounds/MSS_IcicleLarva_GroundEffects.MSS_IcicleLarva_GroundEffects" _rootcomponent))
            (Variables|Default|SetSurfaceAudioComponent _returnvalue))))
      (:else))
    (else
      (bind _surfaceaudiocomponent (Variables|Default|GetSurfaceAudioComponent)
        (:then
          (Audio|Components|Audio|FadeOut _surfaceaudiocomponent 0.4 0.0 "Linear" :!class /Script/Engine.AudioComponent))
        (:else)))))

(event Custom|UpdateSurfaceAudio
  (params (bHitWater :type Boolean))
  (bind _surfaceaudiocomponent (Variables|Default|GetSurfaceAudioComponent)
    (:then
      (Audio|Parameter|SetBooleanParameter _surfaceaudiocomponent "OnWaterHit" bHitWater :!class /Script/AudioExtensions.AudioParameterControllerInterface))
    (:else)))

(fn CalculateRotation
  (params (HitNormal :type Vector%20%28by%20ref%29) (OwnerRotation :type Rotator%20%28by%20ref%29))
  ; Keep owner rotation's yaw, but adjust to hit normal
  (return (Math|Conversions|ToRotator(Quat) (* (Math|Quat|FindQuatBetweenNormals "0.000000,0.000000,1.000000" HitNormal) (Math|Rotator|MakeRotator 0.0 0.0 (.yaw OwnerRotation))))))

(fn DrawDebugShapes
  (bind _returnvalue (Transformation|GetActorLocation))
  (Variables|Default|GetbDebugShapesEnabled
    (:then
      (Rendering|Debug|DrawDebugSphere self _returnvalue 1e+02 16 "(R=1.000000,G=0.003800,B=0.000000,A=1.000000)" 0.0 1.0)
      (Rendering|Debug|DrawDebugArrow self _returnvalue (+ _returnvalue (* (Math|Rotator|GetRotationXVector (Transformation|GetActorRotation)) 2e+02)) 3e+01 "(R=1.000000,G=0.000000,B=0.734788,A=1.000000)" 0.0 1e+01))
    (:else)))

(fn ConstructionScript)
